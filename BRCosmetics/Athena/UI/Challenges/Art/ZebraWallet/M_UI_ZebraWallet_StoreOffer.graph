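# Translucent UI material for the ZebraWallet store offer, compositing a left and a right character image (male and female by default).
# Each image has its own scale and animatable horizontal offset, and their summed alphas form a saturated mask.
# An animatable radial shadow tints the left image before the right is blended over it using its own RGB as weight, not alpha.
# A slight FadeColor tint ramps in toward the bottom and clears as AnimateTint rises.
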Material /BRCosmetics/Athena/UI/Challenges/Art/ZebraWallet/M_UI_ZebraWallet_StoreOffer
graph {
 properties {blendMode=BLEND_Translucent, materialDomain=MD_UI, pixelDepthOffsetMode=PDOM_Legacy, usedShadingModels=Default Lit}
 nodes [
  n1 = TextureCoordinate() {vTiling=0.5}
  n2 = AppendVector(
   A=LinearInterpolate(
    A=ScalarParameter() {defaultValue=0.125, parameterName=U_Translate_LeftImage}
    B=ScalarParameter() {defaultValue=0.15, parameterName=U_Translate_LeftImage_Animated}
    Alpha=ScalarParameter() {parameterName=Anim_Translate}
   )
   B=ScalarParameter() {defaultValue=0.11, parameterName=V_Translate_LeftImage}
  )
  n3 = ScalarParameter() {defaultValue=0.75, parameterName=U_Scale_LeftImage}
  n4 = ScalarParameter() {defaultValue=0.75, parameterName=V_Scale_LeftImage}
  n5 = TextureSampleParameter2D(UVs=MaterialFunctionCall(UVs=Add(A=@n1, B=@n2), Texture Scale=AppendVector(A=@n3, B=@n4)) {materialFunction=&ScaleUVsByCenter}.UVs) {parameterName=LeftTexture, texture=&/T_UI_ZebraWallet_Male_StoreLink}
  n6 = TextureCoordinate() {vTiling=0.5}
  n7 = AppendVector(
   A=LinearInterpolate(
    A=ScalarParameter() {defaultValue=-0.125, parameterName=U_Translate_Right}
    B=ScalarParameter() {defaultValue=-0.15, parameterName=U_Translate_RightImage_Animated}
    Alpha=ScalarParameter() {parameterName=Anim_Translate}
   )
   B=ScalarParameter() {defaultValue=0.11, parameterName=V_Translate_RightImage}
  )
  n8 = ScalarParameter() {defaultValue=0.8, parameterName=U_Scale_RightImage}
  n9 = ScalarParameter() {defaultValue=0.8, parameterName=V_Scale_RightImage}
  n10 = TextureSampleParameter2D(UVs=MaterialFunctionCall(UVs=Add(A=@n6, B=@n7), Texture Scale=AppendVector(A=@n8, B=@n9)) {materialFunction=&ScaleUVsByCenter}.UVs) {parameterName=RightTexture, texture=&/T_UI_ZebraWallet_Female_StoreLink}
  Reroute(None=Saturate(None=Add(A=@n5.A, B=@n10.A)))
  n11 = TextureCoordinate() {vTiling=0.8}
  n12 = AppendVector(
   A=LinearInterpolate(
    A=ScalarParameter() {defaultValue=0.75, parameterName=U_Shadow_Origin}
    B=ScalarParameter() {defaultValue=0.875, parameterName=U_Shadow_Animate_To}
    Alpha=ScalarParameter() {parameterName=AnimateShadow}
   )
   B=ScalarParameter() {defaultValue=0.6565, parameterName=V_Shadow_Origin}
  )
  n13 = ScalarParameter() {defaultValue=0.5619, parameterName=ShadowRadius}
  n14 = ScalarParameter() {defaultValue=5, parameterName=ShadowDensity}
  n15 = LinearInterpolate(
   A=VectorParameter() {parameterName=ShadowColor, defaultValue=<rgba 0.0423 0.0382 0.1046 1>}.RGB
   B=LinearInterpolate(B=@n5.RGB, Alpha=@n5.RGB)
   Alpha=OneMinus(None=MaterialFunctionCall(UVs=@n11, CenterPosition=@n12, Radius=@n13, Density=@n14) {materialFunction=&RadialGradientExponential}.RadialGradientExponential)
  )
  n16 = LinearInterpolate(A=Constant() {r=0.825}, B=Constant() {r=1}, Alpha=ScalarParameter() {parameterName=AnimateTint})
  n17 = Constant() {r=1}
  n18 = ComponentMask(None=TextureCoordinate()) {g=true}
  n19 = Saturate(None=Multiply(A=OneMinus(None=@n18)) {constB=3.333})
  Reroute(None=LinearInterpolate(
    A=VectorParameter() {parameterName=FadeColor, defaultValue=<rgba 0.2195 0.4342 0.7157 1>}.RGB
    B=LinearInterpolate(A=@n15, B=@n10.RGB, Alpha=@n10.RGB)
    Alpha=LinearInterpolate(A=@n16, B=@n17, Alpha=@n19)
   ))
 ]
}
